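# Unlit modulate material for Niagara sprites in the TargetedBeam impact effects, built to scale the scene color behind it by a radial mask.
# A soft SphereMask centered on the UVs blends between inverted and normal via the dynamic parameter Switch, then is shaped by a Pow exponent.
# The result is tinted by particle color and offset through an Add, so particles control the strength and shape of the modulation.
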Material /TargetedBeam/Assets/FX/Impact/M_CamPost
graph {
 properties {
  bAutomaticallySetUsageInEditor=false
  bUsedWithNiagaraSprites=true
  blendMode=BLEND_Modulate
  shadingModel=MSM_Unlit
  usedShadingModels=Unlit
 }
 nodes [
  n1 = Constant() {r=0.5}
  n2 = SphereMask(A=TextureCoordinate(), B=@n1, Radius=@n1) {hardnessPercent=0}
  n3 = DynamicParameter() {defaultValue=<rgba 1 2 1 1>, paramNames [Switch,Pow,Param3,Param4]}
  n4 = LinearInterpolate(A=OneMinus(None=@n2), B=@n2, Alpha=@n3.Switch)
  Add(A=Multiply(A=ParticleColor().RGB, B=Power(Base=@n4, Exp=@n3.Pow)))
 ]
}
